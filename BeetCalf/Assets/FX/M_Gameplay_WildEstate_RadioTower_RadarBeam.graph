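# Translucent, unlit, two-sided material for the radar beam on the WildEstate radio tower.
# Scrolling UVs, offset by a random-scalar function, suit the customized-UV slots and drive two multiplied samples of a beam pattern, boosted and squared for the glow.
# Custom primitive data "Activation Complete" blends two emissive colors; beam and noise textures form a faint mask.
# A camera-perpendicular cross product, scaled by Width and object scale, builds a billboard-style widening offset.
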Material /BeetCalf/Assets/FX/M_Gameplay_WildEstate_RadioTower_RadarBeam
graph {
 properties {
  blendMode=BLEND_Translucent
  numCustomizedUVs=6
  pixelDepthOffsetMode=PDOM_Legacy
  shadingModel=MSM_Unlit
  twoSided=true
  usedShadingModels=Unlit
 }
 nodes [
  Add()
  n1 = Add(A=TextureCoordinate(), B=MaterialFunctionCall() {materialFunction=&/MF_WildEstate_RadioTower_Random}.Random Scalar)
  n2 = ComponentMask(None=@n1) {r=true}
  n3 = Multiply(A=@n2) {constB=0.6}
  n4 = Time()
  n5 = ComponentMask(None=@n1) {g=true}
  n6 = Multiply(A=@n5) {constB=0.15}
  n7 = Multiply(A=@n4) {constB=0.03}
  AppendVector(A=Subtract(A=@n3, B=Multiply(A=@n4) {constB=0.17}), B=Add(A=@n6, B=@n7))
  n8 = Multiply(A=@n2)
  n9 = Time()
  n10 = Multiply(A=@n5) {constB=0.3}
  n11 = Multiply(A=@n9) {constB=0.03}
  AppendVector(A=Subtract(A=@n8, B=Multiply(A=@n9) {constB=0.15}), B=Subtract(A=@n10, B=@n11))
  AppendVector(A=Subtract(A=@n3, B=Multiply(A=@n4) {constB=0.05}), B=Add(A=@n6, B=@n7))
  AppendVector(A=Subtract(A=@n8, B=Multiply(A=@n9) {constB=0.047}), B=Subtract(A=@n10, B=@n11))
  n12 = TextureSample(UVs=TextureCoordinate()) {
   samplerSource=SSM_Clamp_WorldGroupSettings
   samplerType=SAMPLERTYPE_LinearGrayscale
   texture=&/T_Gameplay_WildEstate_RadioTower_Radar_Beam
  }
  n13 = TextureSample(UVs=TextureCoordinate() {coordinateIndex=5}) {samplerType=SAMPLERTYPE_LinearGrayscale, texture=&/T_Gameplay_WildEstate_RadioTower_Radar_Random_A}
  Multiply(A=Multiply(A=@n12.R) {constB=0.23}, B=OneMinus(None=Multiply(A=@n13.R) {constB=0.4}))
  Multiply(A=Add(A=Time(), B=MaterialFunctionCall() {materialFunction=&MF_WildEstate_RadioTower_Random}.Random Scalar)) {constB=0.1}
  n14 = TextureCoordinate() {coordinateIndex=3}
  n15 = TextureCoordinate() {coordinateIndex=4}
  n16 = Multiply(A=Multiply(
    A=TextureSample(UVs=@n14) {samplerType=SAMPLERTYPE_LinearGrayscale, texture=&/T_Gameplay_WildEstate_RadioTower_Radar_Beam_P}.R
    B=TextureSample(UVs=@n15) {samplerType=SAMPLERTYPE_LinearGrayscale, texture=&T_Gameplay_WildEstate_RadioTower_Radar_Beam_P}.R
   )) {constB=20}
  Multiply(
   A=LinearInterpolate(
    A=VectorParameter() {parameterName=Emissive_Color_A, defaultValue=<rgba 0 0.849 0.0475 1>}.RGB
    B=VectorParameter() {parameterName=Emissive_Color_B, defaultValue=<rgba 0 0.849 0.0475 1>}.RGB
    Alpha=ScalarParameter() {bUseCustomPrimitiveData=true, parameterName=Activation Complete, primitiveDataIndex=2}
   )
   B=Add(A=Multiply(A=@n16, B=@n16))
  )
  n17 = WorldPosition()
  n18 = CameraPositionWS()
  n19 = CrossProduct(A=Transform(None=Constant3Vector() {constant=<rgba 1 0 0 0>}), B=Normalize(VectorInput=Subtract(A=@n17.XYZ, B=@n18)))
  n20 = ComponentMask(None=TextureCoordinate()) {g=true}
  Multiply(
   A=Multiply(A=Normalize(VectorInput=@n19), B=Add(A=@n20) {constB=-0.5})
   B=Multiply(
    A=ScalarParameter() {defaultValue=600, parameterName=Width}
    B=MaterialFunctionCall() {materialFunction=&ObjectScale}.Scale X
   )
  ) {constB=25}
  TextureCoordinate()
  VectorParameter() {parameterName=1, defaultValue=<rgba 0.2688 0.8466 1.2 1>}
 ]
}
